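; Swarm-thrower mound that spawns Living World AI encounters and rewards players for killing the swarm.
; Without cosmetics (dedicated server), it optionally starts an initial encounter; a fast pawn lacking blocked tags entering SpawnCollision starts the swarm through five child-actor point providers.
; Every tenth death hides an egg sac; the fourth pauses the encounter and tosses Golden Llama loot.
; OnRep functions play Niagara bursts and hide sacs on clients.
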(event EventBeginPlay
  (Networking|CanExecuteCosmeticEvents
    (:True
      (Components|DestroyComponent (Variables|Default|GetSpawnCollision)))
    (:False
      (Variables|InternalData|SetEncounterInProgress false)
      (Class|FortAsyncActionCurrentPlaylistReady|PlaylistDataReadyAsync
        (:OnReady
          ; If we should spawn an Initial Encounter, do it
          (if (Variables|Default|GetUseInitialEncounter)
            (Utilities|FlowControl|Delay 0.2)
            (bind _output_get (Variables|Default|SetStaticPointProviders (CallFunction|GetProviders)))
            (CallFunction|EnableAllProviders _output_get)
            (bind (_isvalid _returnvalue) (Default|StartActivityEncounterStaticPointProvider (Game|GetGameState) self (Variables|Default|GetInitialLagerEncounter) (Variables|Default|GetStaticPointProviders)))
            (if _isvalid
              (bind _output_get_2 (Variables|InternalData|SetInitialEncounterRef _returnvalue))
              (Variables|InternalData|SetInitialEncounterSpawned true)
              (EventDispatchers|BindEventtoOnLastActorDied _output_get_2 (EventDispatchers|CreateEvent :!event Custom|InitialEncounterLastAIDied))
              (else
                (Fortnite|FortShippingLog self "Tried to Spawn Initial Encounter for Mound, but Encounter failed to start." true)))))))))

(event OnComponentBeginOverlap(SpawnCollision)
  (params (OverlappedComponent :type Primitive%20Component%20Object%20Reference) (OtherActor :type Actor%20Object%20Reference) (OtherComp :type Primitive%20Component%20Object%20Reference) (OtherBodyIndex :type Integer) (bFromSweep :type Boolean) (SweepResult :type Hit%20Result%20Structure%20%28by%20ref%29))
  ; Try Spawn AI: ToDo - Track Velocity from Pawns Inside Volume, not just when entering
  (bind _asfort_pawn (Utilities|Casting|CastToFortPawn OtherActor))
  (if (Math|Boolean|ANDBoolean (>= (Math|Vector|VectorLength (Transformation|GetVelocity _asfort_pawn)) 4e+02) (not (GameplayTags|HasAnyMatchingGameplayTags _asfort_pawn (Variables|Default|GetTagContainerToNotSpawnAI))) (not (Variables|InternalData|GetEncounterInProgress)))
    (Variables|InternalData|SetEncounterInProgress true)
    (if (not (and (Variables|Default|GetUseInitialEncounter) (Variables|InternalData|GetInitialEncounterSpawned)))
      (bind _output_get (Variables|Default|SetStaticPointProviders (CallFunction|GetProviders)))
      (CallFunction|EnableAllProviders _output_get)
      (Utilities|FlowControl|DelayUntilNextTick))
    (CallFunction|ProviderRegistered 0)))

(event Custom|ProviderRegistered
  (params (PointProviderVolume :type Fort%20Athena%20Living%20World%20Volume%20Object%20Reference))
  ; Start Encounter for AI
  (Utilities|FlowControl|Delay 0.2)
  (bind (_isvalid _returnvalue) (Default|StartActivityEncounterStaticPointProvider (Game|GetGameState) self (Utilities|ResolveSoftReference (Variables|Default|GetLagerEncounter)) (Variables|Default|GetStaticPointProviders)))
  (if _isvalid
    (bind _output_get (Variables|InternalData|SetEncounterRef _returnvalue))
    (Class|FortAthenaLivingWorldManager|RequestEventGeneration (Class|FortAthenaLivingWorldManager|GetLivingWorldManager))
    (EventDispatchers|BindEventtoOnActorDied _output_get (EventDispatchers|CreateEvent :!event Custom|SwarmAIDied))
    (else
      (Fortnite|FortShippingLog self "Tried to Spawn Encounter for Mound, but Encounter failed to start." true))))

(event Custom|SwarmAIDied
  (params (EncounterInstance :type Fort%20Athena%20Living%20World%20Encounter%20Instance%20Object%20Reference) (DiedActor :type Actor%20Object%20Reference))
  ; Track Swarm Deaths and Complete Encounter: ToDo - Encounter is just PAUSED and not Ended. Make sure we end it when last creature is killed
  (bind _result (Math|Integer|IncrementInt (Variables|InternalData|GetDeathCount))
    (:%20%20
      (if (and (== (Math|Integer|%25%28Integer%29 _result 10) 0) (not (Variables|InternalData|GetObjectiveCompleted)))
        (CallFunction|SetEggDatatoExplode (Utilities|Select :Index (Variables|InternalData|GetEggSacCounter)))
        (Utilities|FlowControl|Delay 0.2)
        (Variables|InternalData|SetEggSacCounter (+ (Variables|InternalData|GetEggSacCounter) 1))
        (Rendering|SetVisibility (Variables|InternalData|GetEggToExplode) false false)
        (if (>= (Variables|InternalData|GetEggSacCounter) 4)
          (Class|FortAthenaLivingWorldEncounterInstance|PauseEncounter (Variables|InternalData|GetEncounterRef))
          (Inventory|SpawnPickupInWorldWithLootTier "Loot_Golden_Llama" (+ (+ (Transformation|GetActorLocation self) "0.000000,0.000000,650.000000") (* (Transformation|GetActorForwardVector self) 1.5e+02)) 0 0 true true "0.000000,0.000000,300.000000" "Tossed" "ItemSpawner")
          (Variables|InternalData|SetObjectiveCompleted true))))))

(event Custom|InitialEncounterLastAIDied
  (params (EncounterInstance :type Fort%20Athena%20Living%20World%20Encounter%20Instance%20Object%20Reference))
  ; End the Initial Encounter - If there was one
  (Class|BPFLLagerCommunFunctions|EndActivityEncounter (Game|GetGameState) (Variables|InternalData|GetInitialEncounterRef)))

(fn EnableAllProviders
  (params (Providers :type Array%20of%20Fort%20Athena%20Living%20World%20Static%20Point%20Provider%20Object%20References) (Enable :type Boolean :default true))
  (for _array_element Providers
    (if Enable
      (Class|FortAthenaLivingWorldPointProviderInterface|EnablePointProvider_DEPRECATED _array_element)
      (else
        (Class|FortAthenaLivingWorldStaticPointProvider|DisablePointProvider _array_element)))))

(fn GetProviders
  (return (Utilities|Array|MakeArray (Utilities|Casting|CastToBP_StaticPointProvider_Thrower (Variables|/Script/Engine.ChildActorComponent|GetChildActor (Variables|Default|GetBP_StaticPointProvider_Thrower))) (Utilities|Casting|CastToBP_StaticPointProvider_Thrower (Variables|/Script/Engine.ChildActorComponent|GetChildActor (Variables|Default|GetBP_StaticPointProvider_Thrower4))) (Utilities|Casting|CastToBP_StaticPointProvider_Thrower (Variables|/Script/Engine.ChildActorComponent|GetChildActor (Variables|Default|GetBP_StaticPointProvider_Thrower3))) (Utilities|Casting|CastToBP_StaticPointProvider_Thrower (Variables|/Script/Engine.ChildActorComponent|GetChildActor (Variables|Default|GetBP_StaticPointProvider_Thrower2))) (Utilities|Casting|CastToBP_StaticPointProvider_Thrower (Variables|/Script/Engine.ChildActorComponent|GetChildActor (Variables|Default|GetBP_StaticPointProvider_Thrower1))))))

(fn OnRep_EggSacCounter
  (Networking|CanExecuteCosmeticEvents
    (:True
      (Niagara|SpawnSystematLocation :SystemTemplate "/ProtoSwarm/Assets/FX/NS_AvacadoEater_Bird_Destroyed.NS_AvacadoEater_Bird_Destroyed" :Location (Variables|InternalData|GetEggSacExplosionLocation) :Scale "1.000000,1.000000,1.000000" :bAutoDestroy true :bAutoActivate true :PoolingMethod "None" :bPreCullCheck false)
      (Rendering|SetVisibility (Variables|InternalData|GetEggToExplode) false false))))

(fn OnRep_EncounterInProgress
  (Networking|CanExecuteCosmeticEvents
    (:True
      (if (Variables|InternalData|GetEncounterInProgress)
        (Niagara|SpawnSystematLocation "/ProtoSwarm/Assets/FX/NS_CentralPicnic_Milkshake_Burst.NS_CentralPicnic_Milkshake_Burst" (Transformation|GetActorLocation) (Transformation|GetActorRotation) "1.000000,1.000000,1.000000" true true "None" false)))))

(fn SetEggDataToExplode
  (params (EggToExplode :type Static%20Mesh%20Component%20Object%20Reference))
  (bind _output_get (Variables|InternalData|SetEggToExplode EggToExplode))
  (Variables|InternalData|SetEggSacExplosionLocation (Transformation|GetWorldLocation _output_get)))

(fn ConstructionScript)
